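# Material function that gives vehicle wheel rims a chrome look, layered over MF_WheelParent_Simple.
# An angular gradient from squared, offset-adjustable centered UVs combines with object orientation and transforms to camera space, producing view-dependent coordinates for a chrome mask texture.
# The sharpened, brightened sample replaces base color where the rim mask's red channel applies, with emissive zeroed and metallic raised.
# Vertex color red restores the parent's original attributes.
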MaterialFunction /VehicleCosmetics/Wheels/Chrome/Materials/MF_Wheel_Chrome
graph {nodes [
  n1 = MaterialFunctionCall() {materialFunction=&MF_WheelParent_Simple}
  n2 = GetMaterialAttributes(None=@n1.MaterialAttributes) {attributeGetTypes [#1, #2, #3]}
  n3 = ComponentMask(None=@n1.RimRGBAMask) {r=true}
  n4 = LinearInterpolate(A=@n2.EmissiveColor, Alpha=@n3) {constB=0}
  n5 = VertexColor()
  n6 = Subtract(A=TextureCoordinate()) {constB=0.5}
  n7 = AppendVector(
   A=Add(A=ComponentMask(None=@n6) {r=true}, B=ScalarParameter() {group=Chrome, parameterName=ChromeOffsetX})
   B=Add(A=ComponentMask(None=@n6) {g=true}, B=ScalarParameter() {group=Chrome, parameterName=ChromeOffsetY})
  )
  n8 = Multiply(A=@n7, B=@n7)
  n9 = Divide(
   A=Arctangent2Fast(Y=ComponentMask(None=@n8) {g=true}, X=ComponentMask(None=@n8) {r=true})
   B=MaterialFunctionCall(Multiplier=Constant() {r=2}) {materialFunction=&Pi}.Result
  )
  n10 = OneMinus(None=@n9)
  n11 = Subtract(A=@n9, B=ComponentMask(None=ObjectOrientation()) {g=true, r=true})
  n12 = Transform(None=AppendVector(A=AppendVector(A=@n9, B=@n10), B=ComponentMask(None=@n11) {r=true})) {transformSourceType=TRANSFORMSOURCE_Local, transformType=TRANSFORM_Camera}
  n13 = Power(Base=TextureSample(UVs=ComponentMask(None=@n12) {g=true, r=true}) {samplerType=SAMPLERTYPE_LinearColor, texture=/VehicleCosmetics/Wheels/Chrome/Textures/T_Rim_Chrome_ChomeMask}.RGB) {constExponent=3}
  n14 = LinearInterpolate(A=LinearInterpolate(A=@n2.BaseColor, B=Multiply(A=@n13) {constB=2.4, desc=Intensify Brightness}, Alpha=@n3), B=@n2.BaseColor, Alpha=@n5.R)
  n15 = LinearInterpolate(A=@n2.Metallic, Alpha=@n3)
  FunctionOutput(None=SetMaterialAttributes!(
    MaterialAttributes=@n2.MaterialAttributes
    EmissiveColor=LinearInterpolate(A=@n4, B=@n2.EmissiveColor, Alpha=@n5.R)
    BaseColor=Saturate(None=@n14)
    Metallic=LinearInterpolate(A=@n15, B=@n2.Metallic, Alpha=@n5.R)
   ) {attributeSetTypes [#1, #2, #3]})
 ]}
